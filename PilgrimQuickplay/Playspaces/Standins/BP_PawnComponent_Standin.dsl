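; Pawn component for Pilgrim quickplay stand-ins that triggers instrument animation.
; On BeginPlay it waits for the owner's Pilgrim event router, then calls SetupListeners.
; StartInstrumentAnimation looks up the instrument player-state component directly on the supplied pawn and calls OnManualAnimationStart with the playspace.
; A separate helper resolves that component class through the owning Athena pawn's player state.
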(event EventBeginPlay
  (bind _eventrouter (Pilgrim|PilgrimWaitForEventRouter (Components|GetOwner))
    (:OnEventRouterAvailable
      (CallFunction|SetupListeners _eventrouter))))

(event EventTick
  (params (DeltaSeconds :type Float%20%28single-precision%29)))

(event EventStartInstrumentAnimation
  (params (Pawn :type Pawn%20Object%20Reference) (Playspace :type Sparks%20Music%20Playspace%20Object%20Reference))
  (Class|BPPilgrimInstrumentPlayerStateComponent|OnManualAnimationStart (Actor|GetComponentbyClass Pawn "/PilgrimCore/Player/Components/BP_PilgrimInstrumentPlayerStateComponent.BP_PilgrimInstrumentPlayerStateComponent_C") Playspace))

(fn GetPlayerStateComponent
  (bind _asfort_player_pawn_athena (Utilities|Casting|CastToFortPlayerPawnAthena (Components|GetOwner)))
  (bind _returnvalue (Fortnite|Utilities|AttemptGetPlayerStateforPawn _asfort_player_pawn_athena))
  (return (Actor|GetComponentbyClass _returnvalue "/PilgrimCore/Player/Components/BP_PilgrimInstrumentPlayerStateComponent.BP_PilgrimInstrumentPlayerStateComponent_C")))
